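# Animation Blueprint for the Mechanical Engineer owl pet, targeting the Mechanical Engineer female glider skeleton.
# Each animation update calls the parent implementation, then breaks the skydiving state struct from GetSkydivingState.
# It copies aim pitch and yaw, rightward local acceleration and rightward local velocity into Pet_ variables.
# Only this event logic is present; no AnimGraph here shows how those values drive the pose.
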AnimBlueprint /Game/Characters/Pets/Mechanical_Engineer_Owl/B_MechanicalOwl_Pet_AnimBP
target_skeleton=/Game/Gadgets/Assets/VinderTech_GliderChute/Glider_Mechanical_Engineer_Female/Meshes/Mechanical_Engineer_Female_Glider_Skeleton
graphs [EventGraph {nodes [
   n1 = Event BlueprintUpdateAnimation()
   > Parent: BlueprintUpdateAnimation!(execute, DeltaTimeX:float=@n1.DeltaTimeX) -> (then) {is=title K2Node_CallParentFunction}
   n2 = Break Fort Anim Input Skydiving!(FortAnimInput_Skydiving:"Fort Anim Input Skydiving Structure (by ref)"=Get SkydivingState() -> (SkydivingState:Fort Anim Input Skydiving Structure)) -> (
    ParaGilderRootOffset:Vector=~
    ParaGliderRootRotationOverride:Rotator=~
    bUseParaGlideRootModifier:bool=~
    bIsSkydivingFromLaunchPad:bool=~
    bIsSkydivingFromBus:bool=~
    bIsInVortex:bool=~
    bIsDiving:bool=~
    bIsParachuteOpen:bool=~
    bIsSkydiving:bool=~
    bIsParachuteLeaning:bool=~
    bIsSkydiveLeaning:bool=~
    bIsLeaning:bool=~
    bIsSkydiveDiveMode:bool=~
    bPlaySkydiveDrift:bool=~
    bSkydiveDriftDelayActive:bool=~
    bSkydiveDriftAnimAllowed:bool=~
    bIsGliderRight:bool=~
    bIsGliderCenter:bool=~
    bIsGliderLeft:bool=~
    bIsGliderForward:bool=~
    bIsGliderBack:bool=~
    bHasDeployedAndIsStandingUpright:bool=~
    LocalAccelForward:float=~
    LocalAccelRight:float=~
    LocalVelocityRight:float=~
    SkydiveAimPitch:float=~
    SkydiveAimYaw:float=~
    DeployChuteAnimRate:float=~
    SkydiveDriftAnimRate:float=~
    SkydiveDriftAnimRateCurrent:float=~
    SkydiveFidgetAnimRate:float=~
    SkydiveFidgetAnimRateCurrent:float=~
    SkydiveAdditiveAlpha:float=~
    SkydiveDriftDelay:float=~
    SkydiveDriftAnim:int
    SkydiveDriftAnimMax:int
    LaunchpadAnim:int
    LocalAccelDir:ESkydivingDirection Enum
    DirectionLast:ESkydivingDirection Enum
    PoseAdditive:Anim Sequence Object Reference
   ) {is=title K2Node_BreakStruct}
   > Set Pet_SkydiveAimPitch(Pet_SkydiveAimPitch:double=@n2.SkydiveAimPitch) -> (Output_Get:double)
   > Set Pet_SkydiveAimYaw(Pet_SkydiveAimYaw:double=@n2.SkydiveAimYaw) -> (Output_Get:double)
   > Set Pet_LocalAccelRight(Pet_LocalAccelRight:double=@n2.LocalAccelRight) -> (Output_Get:double)
   > Set Pet_LocalVelocityRight(Pet_LocalVelocityRight:double=@n2.LocalVelocityRight) -> (Output_Get:double)
   TryGetPawnOwner()
  ]}]
variables [
 Pet_LocalAccelRight {type=number}
 Pet_LocalVelocityRight {type=number}
 Pet_SkydiveAimPitch {type=number}
 Pet_SkydiveAimYaw {type=number}
]
